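# Physics asset for the F_MED_Lyrical female player body cosmetic, covering the hood and hoodie laces, apparently for secondary motion.
# Capsule bodies sit on two hood bones, mirrored left and right lace chains, two upper spine bones and the left clavicle only.
# Each lace chain links three segments to its master bone, which has no constraint to the torso.
# Both hood bones constrain directly to spine_05, with tight swing and twist limits throughout.
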PhysicsAsset /BRCosmetics/Characters/Player/Female/Medium/Bodies/F_MED_Lyrical/Meshes/F_MED_Lyrical_Physics
bodies {
 clavicle_l [{shapeType=Capsule, length=1.2059, radius=0.9089, shapeName=clavicle_l_capsule, rotation=<rot 90 180 179.9996>}]
 dyn_hood_01 [{shapeType=Capsule, length=6.5962, radius=5.2091, shapeName=dyn_hood_01_capsule}]
 dyn_hood_02 [{shapeType=Capsule, length=6.4226, radius=5.7298, shapeName=dyn_hood_02_capsule}]
 dyn_hoodie_laces_01_l [{
   shapeType=Capsule
   length=3.8467
   radius=0.5249
   shapeName=dyn_hoodie_laces_01_l_capsule
   center=<3.5921 0 0>
   rotation=<rot 90 180 -179.9998>
  }]
 dyn_hoodie_laces_01_r [{shapeType=Capsule, length=3.8467, radius=0.5249, shapeName=dyn_hoodie_laces_01_l_capsule, center=<-3.5921 0 0>, rotation=<rot 90 180 0.0002>}]
 dyn_hoodie_laces_02_l [{
   shapeType=Capsule
   length=4.7883
   radius=0.5608
   shapeName=dyn_hoodie_laces_02_l_capsule
   center=<2.46 0 0>
   rotation=<rot 90 -179.9605 -179.9603>
  }]
 dyn_hoodie_laces_02_r [{
   shapeType=Capsule
   length=4.7883
   radius=0.5608
   shapeName=dyn_hoodie_laces_02_l_capsule
   center=<-2.46 0 0>
   rotation=<rot 90 -179.9803 0.02>
  }]
 dyn_hoodie_laces_03_l [{
   shapeType=Capsule
   length=5.3167
   radius=0.5608
   shapeName=dyn_hoodie_laces_03_l_capsule
   center=<2.0887 0 0>
   rotation=<rot 90 -179.9208 -179.9205>
  }]
 dyn_hoodie_laces_03_r [{
   shapeType=Capsule
   length=5.3167
   radius=0.5608
   shapeName=dyn_hoodie_laces_03_l_capsule
   center=<-2.0887 0 0>
   rotation=<rot 90 -0.0791 179.9211>
  }]
 dyn_hoodie_laces_mstr_l [{shapeType=Capsule, length=0.6003, radius=0.6554, shapeName=dyn_hoodie_laces_mstr_l_capsule}]
 dyn_hoodie_laces_mstr_r [{shapeType=Capsule, length=0.6003, radius=0.6554, shapeName=dyn_hoodie_laces_mstr_l_capsule, rotation=<rot 0 0 -180>}]
 spine_04 [{shapeType=Capsule, length=3.178, radius=11.9155, shapeName=spine_05_capsule, center=<-1.2214 -0.15 0>}]
 spine_05 [{shapeType=Capsule, length=4.0732, radius=11.7691, shapeName=spine_05_capsule, center=<-5.4258 0.2819 0>}]
}
constraints [
 {bone1Name=dyn_hoodie_laces_03_l, bone2Name=dyn_hoodie_laces_02_l, swing1LimitDegrees=20, swing2LimitDegrees=20, twistLimitDegrees=5}
 {bone1Name=dyn_hoodie_laces_02_l, bone2Name=dyn_hoodie_laces_01_l, swing1LimitDegrees=20, swing2LimitDegrees=20, twistLimitDegrees=5}
 {bone1Name=dyn_hoodie_laces_01_l, bone2Name=dyn_hoodie_laces_mstr_l, swing1LimitDegrees=20, swing2LimitDegrees=20, twistLimitDegrees=5}
 {bone1Name=dyn_hoodie_laces_03_r, bone2Name=dyn_hoodie_laces_02_r, swing1LimitDegrees=20, swing2LimitDegrees=20, twistLimitDegrees=5}
 {bone1Name=dyn_hoodie_laces_02_r, bone2Name=dyn_hoodie_laces_01_r, swing1LimitDegrees=20, swing2LimitDegrees=20, twistLimitDegrees=5}
 {bone1Name=dyn_hoodie_laces_01_r, bone2Name=dyn_hoodie_laces_mstr_r, swing1LimitDegrees=20, swing2LimitDegrees=20, twistLimitDegrees=5}
 {bone1Name=dyn_hood_02, bone2Name=spine_05, swing1LimitDegrees=20, swing2LimitDegrees=15, twistLimitDegrees=5}
 {bone1Name=dyn_hood_01, bone2Name=spine_05, swing1LimitDegrees=25, swing2LimitDegrees=15, twistLimitDegrees=5}
]
