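# Physics asset for the face accessory mesh of the F_MED_Lyrical female player cosmetic, defining the physics bodies for its ponytail.
# Capsules on the ponytail master bone and three segment bones form a chain, with each joint using narrow swing limits and a tighter twist limit.
# A wide, unconstrained capsule on spine_05 adds an upper-torso body alongside the chain.
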PhysicsAsset /BRCosmetics/Characters/Player/Female/Medium/Bodies/F_MED_Lyrical/Meshes/Parts/F_MED_Lyrical_FaceAcc_Physics
bodies {
 dyn_ponytail_01 [{
   shapeType=Capsule
   length=5.1163
   radius=2.347
   shapeName=dyn_ponytail_01_capsule
   center=<3.3541 0.0001 0>
   rotation=<rot -0.0001 90.0011 90>
  }]
 dyn_ponytail_02 [{
   shapeType=Capsule
   length=5.1163
   radius=2.347
   shapeName=dyn_ponytail_02_capsule
   center=<3.3541 0.0001 0>
   rotation=<rot 0.0001 90.0011 90>
  }]
 dyn_ponytail_03 [{
   shapeType=Capsule
   length=5.0089
   radius=2.4543
   shapeName=dyn_ponytail_03_capsule
   center=<2.621 -0.4999 0.2616>
   rotation=<rot -0.5255 76.0186 94.9724>
  }]
 dyn_ponytail_mstr [{shapeType=Capsule, length=0.5592, radius=0.7607, shapeName=dyn_ponytail_mstr_capsule, rotation=<rot -0.0001 90.0011 90>}]
 spine_05 [{shapeType=Capsule, length=4.5708, radius=12.4153, shapeName=spine_05_capsule, center=<7.3346 1.0105 0>}]
}
constraints [
 {bone1Name=dyn_ponytail_03, bone2Name=dyn_ponytail_02, swing1LimitDegrees=30, swing2LimitDegrees=20, twistLimitDegrees=5}
 {bone1Name=dyn_ponytail_02, bone2Name=dyn_ponytail_01, swing1LimitDegrees=20, swing2LimitDegrees=30, twistLimitDegrees=5}
 {bone1Name=dyn_ponytail_01, bone2Name=dyn_ponytail_mstr, swing1LimitDegrees=20, swing2LimitDegrees=30, twistLimitDegrees=5}
]
